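; Cosmetic glider Blueprint for BlackMondayCake that drives trail particles and thrust audio from parachute movement.
; Trail updates set thrust audio by forward motion and, once fully deployed, activate trails when moving forward or backward, using paired DoOnce gates that reset each other.
; AirFoilCheck sets left/right trail alpha from the right dot, behind a DoOnce.
; BeginPlay passes the separate-translucency console setting to the particles.
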(event EventBeginPlay
  (|CallParentFunction)
  (bind _returnvalue (Development|GetConsoleVariableBoolValue "r.SeparateTranslucency"))
  (Effects|Components|ParticleSystem|SetFloatParameter (Variables|Parachute|GetTrailParticles) "SepTrans" (select _returnvalue 0.0 1.0)))

(event Custom|Reset
  (Utilities|FlowControl|DoOnce false
    (:Completed
      (Components|Activation|Activate (Variables|Parachute|GetTrailParticles) false)
      (CallFunction|ResetReverse))))

(event Parachute|EventOnParachuteTrailUpdated
  (params (MovementLengthSquared :type Float%20%28single-precision%29) (ForwardDot :type Float%20%28single-precision%29) (RightDot :type Float%20%28single-precision%29) (RotationalMovementAcceleration :type Float%20%28single-precision%29))
  (bind _returnvalue (> ForwardDot 0.2))
  (bind _returnvalue_2 (CallFunction|IsGliderFullyDeployedAccordingtoAnim))
  (Variables|Default|SetRight%20DOT RightDot)
  (CallFunction|AudioUpdate _returnvalue)
  (if _returnvalue_2
    (if (and (or _returnvalue (< ForwardDot -0.1)) _returnvalue_2)
      (CallFunction|AirFoilCheck (Variables|Default|GetRight%20DOT))
      (Utilities|FlowControl|DoOnce false
        (:Completed
          (Components|Activation|Activate (Variables|Parachute|GetTrailParticles) false)
          (CallFunction|ResetReverse)))
      (else
        (Utilities|FlowControl|DoOnce true
          (:Completed
            (Components|Activation|Deactivate (Variables|Parachute|GetTrailParticles))
            (CallFunction|Reset)))))
    (else
      (Components|Activation|Deactivate (Variables|Parachute|GetTrailParticles)))))

(event Custom|Reset_Reverse
  (Utilities|FlowControl|DoOnce true
    (:Completed
      (Components|Activation|Deactivate (Variables|Parachute|GetTrailParticles))
      (CallFunction|Reset))))

(event Custom|Audio%20Update
  (params (Is%20Moving%20Forward :type Boolean))
  (bind _thrustaudio (Variables|Default|GetThrustAudio))
  ;[1 1 1] Update audio when thrusting
  (if (CallFunction|IsGliderFullyDeployedAccordingtoAnim)
    (Class|FortVehicleAudioVoice|SetParam _thrustaudio (select Is%20Moving%20Forward 1.0 0.5))
    (else
      (Class|FortVehicleAudioVoice|SetParam _thrustaudio 0.0))))

(event EventOnGliderFullyDeployed)

(fn AirFoilCheck
  (params (Right%20Dot :type Float%20%28double-precision%29))
  (bind _trailparticles (Variables|Parachute|GetTrailParticles))
  (bind _trailparticles_2 (Variables|Parachute|GetTrailParticles))
  (bind _trailparticles_3 (Variables|Parachute|GetTrailParticles))
  (if (not (Math|Float|NearlyEqual(Float) Right%20Dot (Variables|Default|GetRight%20DOT)))
    (Utilities|FlowControl|Sequence
      (:then_0
        _)
      (:then_1
        _)))
  (Utilities|FlowControl|DoOnce false
    (:Completed
      (if (>= (Variables|Default|GetRight%20Dot) 0.1)
        (Effects|Components|ParticleSystem|SetFloatParameter _trailparticles_2 "TrailAlphaLeft" 2.0)
        (Effects|Components|ParticleSystem|SetFloatParameter _trailparticles_2 "TrailAlphaRight" 0.5)
        (elif (<= (Variables|Default|GetRight%20Dot) -0.1)
          (Effects|Components|ParticleSystem|SetFloatParameter _trailparticles_3 "TrailAlphaLeft" 0.5)
          (Effects|Components|ParticleSystem|SetFloatParameter _trailparticles_3 "TrailAlphaRight" 2.0)
          (else
            (Effects|Components|ParticleSystem|SetFloatParameter _trailparticles "TrailAlphaLeft" 1.0)
            (Effects|Components|ParticleSystem|SetFloatParameter _trailparticles "TrailAlphaRight" 1.0)))))))

(fn ConstructionScript)
